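# Physics asset for the Virtuous backpack cosmetic, defining capsule bodies and joints for its pair of feathered wings.
# Free-rotating joints link root, pack, backpack root and wing shoulders, then constrained chains run through elbow and wrist into fingers and feathers.
# Zero twist limits and mostly single-axis swing keep wings stiff, while shoulders and F-feather chains bend on two axes.
# Root also carries a large, offset, tilted capsule, likely a body collider.
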PhysicsAsset /BRCosmetics/Accessories/FORT_Backpacks/Backpack_F_MED_Virtuous/Meshes/F_MED_Virtuous_Pack_Physics
bodies {
 backpack_root [{shapeType=Capsule, length=0.5, radius=0.5, shapeName=backpack_root_capsule, rotation=<rot 90 0 0>}]
 dyn_featherA_02_l [{shapeType=Capsule, length=3, radius=2, shapeName=dyn_featherA_02_l_capsule, center=<4 0 0>, rotation=<rot 90 0 0>}]
 dyn_featherA_02_r [{shapeType=Capsule, length=3, radius=2, shapeName=dyn_featherA_02_r_capsule, center=<4 0 0>, rotation=<rot 90 0 0>}]
 dyn_featherA_03_l [{shapeType=Capsule, length=3, radius=2, shapeName=dyn_featherA_03_l_capsule, center=<5 0 0>, rotation=<rot 90 0 0>}]
 dyn_featherA_03_r [{shapeType=Capsule, length=3, radius=2, shapeName=dyn_featherA_03_r_capsule, center=<5 0 0>, rotation=<rot 90 0 0>}]
 dyn_featherB_01_l [{shapeType=Capsule, length=3, radius=1, shapeName=dyn_featherB_01_l_capsule, center=<5 0 0>, rotation=<rot 90 0 0>}]
 dyn_featherB_01_r [{shapeType=Capsule, length=3, radius=1, shapeName=dyn_featherB_01_l_capsule, center=<5 0 0>, rotation=<rot 90 0 0>}]
 dyn_featherB_02_l [{shapeType=Capsule, length=3, radius=1, shapeName=dyn_featherB_02_l_capsule, center=<3 0 0>, rotation=<rot 90 0 0>}]
 dyn_featherB_02_r [{shapeType=Capsule, length=3, radius=1, shapeName=dyn_featherB_02_r_capsule, center=<3 0 0>, rotation=<rot 90 0 0>}]
 dyn_featherC_02_l [{shapeType=Capsule, length=3, radius=0.8, shapeName=dyn_featherC_02_l_capsule, center=<3 0 0>, rotation=<rot 90 0 0>}]
 dyn_featherC_02_r [{shapeType=Capsule, length=3, radius=0.8, shapeName=dyn_featherC_02_r_capsule, center=<3 0 0>, rotation=<rot 90 0 0>}]
 dyn_featherD_01_l [{shapeType=Capsule, length=3, radius=1, shapeName=dyn_featherD_01_l_capsule, center=<5 0 0>, rotation=<rot 90 0 0>}]
 dyn_featherD_01_r [{shapeType=Capsule, length=3, radius=1, shapeName=dyn_featherD_01_l_capsule, center=<5 0 0>, rotation=<rot 90 0 0>}]
 dyn_featherD_02_l [{shapeType=Capsule, length=3, radius=1, shapeName=dyn_featherD_02_l_capsule, center=<2 0 0>, rotation=<rot 90 0 0>}]
 dyn_featherD_02_r [{shapeType=Capsule, length=3, radius=1, shapeName=dyn_featherD_02_r_capsule, center=<2 0 0>, rotation=<rot 90 0 0>}]
 dyn_featherF_01_l [{shapeType=Capsule, length=13.8143, radius=2.2433, shapeName=dyn_featherF_01_l_capsule, center=<8.24 0 0>, rotation=<rot 90 0 0>}]
 dyn_featherF_01_r [{shapeType=Capsule, length=12.3844, radius=1.9276, shapeName=dyn_featherF_01_l_capsule, center=<7.63 0 0>, rotation=<rot 90 0 0>}]
 dyn_featherF_02_l [{shapeType=Capsule, length=7.4101, radius=2.6222, shapeName=dyn_featherF_02_l_capsule, center=<6.2011 0 0>, rotation=<rot 90 0 0>}]
 dyn_featherF_02_r [{shapeType=Capsule, length=7.4101, radius=2.6222, shapeName=dyn_featherF_02_l_capsule, center=<6.2011 0 0>, rotation=<rot 90 0 0>}]
 dyn_featherF_03_l [{shapeType=Capsule, length=6.9051, radius=1.7484, shapeName=dyn_featherF_03_l_capsule, center=<5.4365 0 0>, rotation=<rot 90 0 0>}]
 dyn_featherF_03_r [{shapeType=Capsule, length=6.9051, radius=1.7484, shapeName=dyn_featherF_03_l_capsule, center=<5.4365 0 0>, rotation=<rot 90 0 0>}]
 dyn_wing_elbow_l [{shapeType=Capsule, length=9.5204, radius=1.9796, shapeName=dyn_wing_elbow_l_capsule, center=<5.628 0 0>, rotation=<rot 90 0 0>}]
 dyn_wing_elbow_r [{shapeType=Capsule, length=8, radius=3, shapeName=dyn_wing_elbow_l_capsule, center=<4.4046 0 0>, rotation=<rot 90 0 180>}]
 dyn_wing_shoulder_l [{shapeType=Capsule, length=2, radius=1.5185, shapeName=dyn_wing_shoulder_l_capsule, center=<3.2457 0 0>, rotation=<rot 90 0 0>}]
 dyn_wing_shoulder_r [{shapeType=Capsule, length=2, radius=2.2781, shapeName=dyn_wing_shoulder_l_capsule, center=<3.2457 0 0>, rotation=<rot 90 0 180>}]
 pack [{shapeType=Capsule, length=0.5, radius=0.5, shapeName=pack_capsule, rotation=<rot 90 0 0>}]
 root [
  {shapeType=Capsule, length=0.5, radius=0.5, shapeName=root_capsule, rotation=<rot 90 0 0>}
  {shapeType=Capsule, length=93.2219, radius=33.7517, shapeName=root_capsule, center=<0 26.6657 -34.1142>, rotation=<rot 0 0 26.415>}
 ]
 wing_index_l [{shapeType=Capsule, length=3, radius=1, shapeName=wing_index_l_capsule, center=<5 0 0>, rotation=<rot 90 0 0>}]
 wing_index_r [{shapeType=Capsule, length=3, radius=1, shapeName=wing_index_l_capsule, center=<5 0 0>, rotation=<rot 90 0 0>}]
 wing_pinky_l [{shapeType=Capsule, length=3, radius=1, shapeName=wing_pinky_l_capsule, center=<3 0 0>, rotation=<rot 90 0 0>}]
 wing_pinky_r [{shapeType=Capsule, length=3, radius=1, shapeName=wing_pinky_l_capsule, center=<3 0 0>, rotation=<rot 90 0 0>}]
 wing_ring_l [{shapeType=Capsule, length=3, radius=1, shapeName=wing_ring_l_capsule, center=<5 0 0>, rotation=<rot 90 0 0>}]
 wing_ring_r [{shapeType=Capsule, length=3, radius=1, shapeName=wing_ring_l_capsule, center=<5 0 0>, rotation=<rot 90 0 0>}]
 wing_shoulder_l [{shapeType=Capsule, length=0.5, radius=0.5, shapeName=wing_shoulder_l_capsule, rotation=<rot 90 0 0>}]
 wing_shoulder_r [{shapeType=Capsule, length=0.5, radius=0.5, shapeName=wing_shoulder_r_capsule, rotation=<rot 90 0 0>}]
 wing_wrist_l [{shapeType=Capsule, length=2, radius=2, shapeName=wing_wrist_l_capsule, center=<2.5919 0 0>, rotation=<rot 90 0 0>}]
 wing_wrist_r [{shapeType=Capsule, length=2, radius=2, shapeName=wing_wrist_l_capsule, center=<1.5 0 0>, rotation=<rot 90 0 180>}]
}
constraints [
 {
  bone1Name=pack
  bone2Name=root
  swing1LimitDegrees=45
  swing1Motion=Free
  swing2LimitDegrees=45
  swing2Motion=Free
  twistLimitDegrees=45
  twistMotion=Free
 }
 {
  bone1Name=backpack_root
  bone2Name=pack
  swing1LimitDegrees=45
  swing1Motion=Free
  swing2LimitDegrees=45
  swing2Motion=Free
  twistLimitDegrees=45
  twistMotion=Free
 }
 {
  bone1Name=wing_shoulder_l
  bone2Name=backpack_root
  swing1LimitDegrees=45
  swing1Motion=Free
  swing2LimitDegrees=45
  swing2Motion=Free
  twistLimitDegrees=45
  twistMotion=Free
 }
 {bone1Name=dyn_wing_shoulder_l, bone2Name=wing_shoulder_l, swing1LimitDegrees=10, swing2LimitDegrees=10, twistLimitDegrees=0}
 {bone1Name=dyn_featherF_01_l, bone2Name=dyn_wing_shoulder_l, swing1LimitDegrees=30, swing2LimitDegrees=30, twistLimitDegrees=0}
 {bone1Name=dyn_featherF_02_l, bone2Name=dyn_featherF_01_l, swing1LimitDegrees=30, swing2LimitDegrees=30, twistLimitDegrees=0}
 {bone1Name=dyn_wing_elbow_l, bone2Name=dyn_wing_shoulder_l, swing1LimitDegrees=0, swing2LimitDegrees=3, twistLimitDegrees=0}
 {bone1Name=wing_wrist_l, bone2Name=dyn_wing_elbow_l, swing1LimitDegrees=0, swing2LimitDegrees=4, twistLimitDegrees=0}
 {bone1Name=dyn_featherB_01_l, bone2Name=wing_wrist_l, swing1LimitDegrees=0, swing2LimitDegrees=5, twistLimitDegrees=0}
 {bone1Name=dyn_featherD_01_l, bone2Name=wing_wrist_l, swing1LimitDegrees=0, swing2LimitDegrees=5, twistLimitDegrees=0}
 {bone1Name=wing_index_l, bone2Name=wing_wrist_l, swing1LimitDegrees=0, swing2LimitDegrees=5, twistLimitDegrees=0}
 {bone1Name=wing_pinky_l, bone2Name=wing_wrist_l, swing1LimitDegrees=0, swing2LimitDegrees=10, twistLimitDegrees=0}
 {bone1Name=wing_ring_l, bone2Name=wing_wrist_l, swing1LimitDegrees=0, swing2LimitDegrees=5, twistLimitDegrees=0}
 {
  bone1Name=wing_shoulder_r
  bone2Name=backpack_root
  swing1LimitDegrees=45
  swing1Motion=Free
  swing2LimitDegrees=45
  swing2Motion=Free
  twistLimitDegrees=45
  twistMotion=Free
 }
 {bone1Name=dyn_wing_elbow_r, bone2Name=dyn_wing_shoulder_r, swing1LimitDegrees=0, swing2LimitDegrees=3, twistLimitDegrees=0}
 {bone1Name=dyn_wing_shoulder_r, bone2Name=wing_shoulder_r, swing1LimitDegrees=12, swing2LimitDegrees=15, twistLimitDegrees=0}
 {bone1Name=wing_wrist_r, bone2Name=dyn_wing_elbow_r, swing1LimitDegrees=0, swing2LimitDegrees=4, twistLimitDegrees=0}
 {bone1Name=wing_ring_r, bone2Name=wing_wrist_r, swing1LimitDegrees=0, swing2LimitDegrees=5, twistLimitDegrees=0}
 {bone1Name=dyn_featherB_01_r, bone2Name=wing_wrist_r, swing1LimitDegrees=0, swing2LimitDegrees=5, twistLimitDegrees=0}
 {bone1Name=dyn_featherD_01_r, bone2Name=wing_wrist_r, swing1LimitDegrees=0, swing2LimitDegrees=5, twistLimitDegrees=0}
 {bone1Name=wing_index_r, bone2Name=wing_wrist_r, swing1LimitDegrees=0, swing2LimitDegrees=5, twistLimitDegrees=0}
 {bone1Name=wing_pinky_r, bone2Name=wing_wrist_r, swing1LimitDegrees=0, swing2LimitDegrees=10, twistLimitDegrees=0}
 {bone1Name=dyn_featherA_02_l, bone2Name=wing_pinky_l, swing1LimitDegrees=0, swing2LimitDegrees=10, twistLimitDegrees=0}
 {bone1Name=dyn_featherA_03_l, bone2Name=dyn_featherA_02_l, swing1LimitDegrees=0, swing2LimitDegrees=10, twistLimitDegrees=0}
 {bone1Name=dyn_featherB_02_l, bone2Name=dyn_featherB_01_l, swing1LimitDegrees=0, swing2LimitDegrees=5, twistLimitDegrees=0}
 {bone1Name=dyn_featherC_02_l, bone2Name=wing_ring_l, swing1LimitDegrees=0, swing2LimitDegrees=5, twistLimitDegrees=0}
 {bone1Name=dyn_featherD_02_l, bone2Name=dyn_featherD_01_l, swing1LimitDegrees=0, swing2LimitDegrees=5, twistLimitDegrees=0}
 {bone1Name=dyn_featherC_02_r, bone2Name=wing_ring_r, swing1LimitDegrees=0, swing2LimitDegrees=5, twistLimitDegrees=0}
 {bone1Name=dyn_featherA_03_r, bone2Name=dyn_featherA_02_r, swing1LimitDegrees=0, swing2LimitDegrees=10, twistLimitDegrees=0}
 {bone1Name=dyn_featherA_02_r, bone2Name=wing_pinky_r, swing1LimitDegrees=0, swing2LimitDegrees=10, twistLimitDegrees=0}
 {bone1Name=dyn_featherD_02_r, bone2Name=dyn_featherD_01_r, swing1LimitDegrees=0, swing2LimitDegrees=5, twistLimitDegrees=0}
 {bone1Name=dyn_featherB_02_r, bone2Name=dyn_featherB_01_r, swing1LimitDegrees=0, swing2LimitDegrees=5, twistLimitDegrees=0}
 {bone1Name=dyn_featherF_03_l, bone2Name=dyn_featherF_02_l, swing1LimitDegrees=30, swing2LimitDegrees=30, twistLimitDegrees=0}
 {bone1Name=dyn_featherF_03_r, bone2Name=dyn_featherF_02_r, swing1LimitDegrees=30, swing2LimitDegrees=30, twistLimitDegrees=0}
 {bone1Name=dyn_featherF_02_r, bone2Name=dyn_featherF_01_r, swing1LimitDegrees=30, swing2LimitDegrees=30, twistLimitDegrees=0}
 {bone1Name=dyn_featherF_01_r, bone2Name=dyn_wing_shoulder_r, swing1LimitDegrees=30, swing2LimitDegrees=30, twistLimitDegrees=0}
]
